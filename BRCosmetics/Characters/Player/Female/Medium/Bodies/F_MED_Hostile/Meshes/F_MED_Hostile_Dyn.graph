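# Physics asset for the dynamic skirt of the F_MED_Hostile female player body.
# Capsules cover three-segment skirt bone chains plus pelvis, lower spine and thighs.
# Chains hang from the pelvis, directly at the back and through rotation-free MSTR bodies elsewhere, with limited or locked swing and tight twist.
# Free-rotation cross constraints link many neighboring hem segments, and the left front seam's top segment is rotation-locked to its neighbor.
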PhysicsAsset /BRCosmetics/Characters/Player/Female/Medium/Bodies/F_MED_Hostile/Meshes/F_MED_Hostile_Dyn
bodies {
 dyn_skirt_bk_01 [{shapeType=Capsule, length=2.5, radius=5.75, shapeName=dyn_skirt_bk_01_capsule, center=<5.0516 0 5>, rotation=<rot 0 0 90>}]
 dyn_skirt_bk_02 [{shapeType=Capsule, length=5, radius=8.25, shapeName=dyn_skirt_bk_02_capsule, center=<5.2235 0 7.7>, rotation=<rot 0 0 90>}]
 dyn_skirt_bk_03 [{shapeType=Capsule, length=11.5, radius=5.75, shapeName=dyn_skirt_bk_03_capsule, center=<5.0403 0 4.1>, rotation=<rot 0 0 90>}]
 dyn_skirt_bk_l_01 [{shapeType=Capsule, length=3.25, radius=9, shapeName=dyn_skirt_bk_l_01_capsule, center=<5.0449 0 8.2>, rotation=<rot 0 0 90>}]
 dyn_skirt_bk_l_02 [{shapeType=Capsule, length=3.25, radius=9.5, shapeName=dyn_skirt_bk_l_02_capsule, center=<5.4171 0 8.7>, rotation=<rot 0 0 -90>}]
 dyn_skirt_bk_l_03 [{
   shapeType=Capsule
   length=15.75
   radius=6.75
   shapeName=dyn_skirt_bk_l_03_capsule
   center=<4.9297 0 4.2>
   rotation=<rot -0.0003 0.0001 103.4374>
  }]
 dyn_skirt_bk_r_01 [{shapeType=Capsule, length=7.25, radius=5.75, shapeName=dyn_skirt_bk_r_01_capsule, center=<5.0758 0 5>, rotation=<rot 0 0 90>}]
 dyn_skirt_bk_r_02 [{shapeType=Capsule, length=15, radius=8, shapeName=dyn_skirt_bk_r_02_capsule, center=<5.5777 0 7.4>, rotation=<rot 0 0 90>}]
 dyn_skirt_bk_r_03 [{
   shapeType=Capsule
   length=15
   radius=15
   shapeName=dyn_skirt_bk_r_03_capsule
   center=<5.6509 0 12.3>
   rotation=<rot -0.0004 -0.0001 -100.2041>
  }]
 dyn_skirt_fr_MSTR [{shapeType=Capsule, length=1, radius=1}]
 dyn_skirt_fr_l_01 [{shapeType=Capsule, length=1.25, radius=6, shapeName=dyn_skirt_fr_l_01_capsule, center=<4.9773 0 5.3>, rotation=<rot 0 0 90>}]
 dyn_skirt_fr_l_02 [{shapeType=Capsule, length=7.25, radius=8.25, shapeName=dyn_skirt_fr_l_02_capsule, center=<4.8899 4 6.7>, rotation=<rot 0 0 90>}]
 dyn_skirt_fr_l_03 [{shapeType=Capsule, length=12.25, radius=5.75, shapeName=dyn_skirt_fr_l_03_capsule, center=<5.5649 3.4 4>, rotation=<rot 0 0 90>}]
 dyn_skirt_fr_mid_l_01 [{shapeType=Capsule, length=2.25, radius=4.75, shapeName=dyn_skirt_fr_mid_l_01_capsule, center=<2.2365 0 4.5>, rotation=<rot 0 87.5698 90>}]
 dyn_skirt_fr_mid_l_02 [
  {shapeType=Capsule, length=9.5, radius=6.5, shapeName=dyn_skirt_fr_mid_l_02_capsule, center=<5.0946 0 6.5>, rotation=<rot 0.0001 0 93.2171>}
  {
   shapeType=Capsule
   length=7.25
   radius=7
   shapeName=dyn_skirt_fr_mid_l_02_capsule
   center=<5.3196 -11.1814 2.0281>
   rotation=<rot 0.0001 0 -121.4098>
  }
 ]
 dyn_skirt_fr_mid_l_03 [
  {
   shapeType=Capsule
   length=8.75
   radius=9
   shapeName=dyn_skirt_fr_mid_l_03_capsule
   center=<5.1945 0 8.6>
   rotation=<rot 0.0003 0.0001 95.0159>
  }
  {
   shapeType=Capsule
   length=10.25
   radius=7
   shapeName=dyn_skirt_fr_mid_l_03_capsule
   center=<5.2251 -10.1531 1.6717>
   rotation=<rot 0.0001 0.0001 -116.4268>
  }
 ]
 dyn_skirt_fr_mid_l_MSTR [{shapeType=Capsule, length=1, radius=1}]
 dyn_skirt_fr_mid_r_01 [{shapeType=Capsule, length=4.75, radius=5.75, shapeName=dyn_skirt_fr_mid_r_01_capsule, center=<5.8161 0 4.9>, rotation=<rot 0 0 90>}]
 dyn_skirt_fr_mid_r_02 [{
   shapeType=Capsule
   length=13.25
   radius=7.25
   shapeName=dyn_skirt_fr_mid_r_02_capsule
   center=<5.2706 0 4>
   rotation=<rot 0.0006 -0.0002 86.0034>
  }]
 dyn_skirt_fr_mid_r_03 [{
   shapeType=Capsule
   length=12
   radius=9.5
   shapeName=dyn_skirt_fr_mid_r_03_capsule
   center=<4.8051 0 4.3>
   rotation=<rot 0.0002 -0.0001 97.9957>
  }]
 dyn_skirt_fr_mid_r_MSTR [{shapeType=Capsule, length=1, radius=1}]
 dyn_skirt_fr_seam_l_01 [{shapeType=Capsule, length=8.5, radius=3.5, shapeName=dyn_skirt_fr_seam_l_01_capsule, center=<5.1858 3.2 1.9>, rotation=<rot 90 0 0>}]
 dyn_skirt_fr_seam_l_02 [{shapeType=Capsule, length=6.25, radius=6, shapeName=dyn_skirt_fr_seam_l_02_capsule, center=<5.2473 1.5 3>, rotation=<rot 90 0 0>}]
 dyn_skirt_fr_seam_l_03 [{shapeType=Capsule, length=5.75, radius=5.25, shapeName=dyn_skirt_fr_seam_l_03_capsule, center=<5.2877 1.9 2>, rotation=<rot 90 0 0>}]
 dyn_skirt_fr_seam_l_MSTR [{shapeType=Capsule, length=1, radius=1}]
 dyn_skirt_l_01 [{shapeType=Capsule, length=3, radius=5.75, shapeName=dyn_skirt_l_01_capsule, center=<5.0078 0 5.1>, rotation=<rot 0 0 90>}]
 dyn_skirt_l_02 [{shapeType=Capsule, length=11, radius=6.75, shapeName=dyn_skirt_l_02_capsule, center=<5.5383 0 5.4>, rotation=<rot 0 0 -73.6762>}]
 dyn_skirt_l_03 [{shapeType=Capsule, length=19.5, radius=6, shapeName=dyn_skirt_l_03_capsule, center=<5.4757 0 4.3>, rotation=<rot 0 0 -70>}]
 dyn_skirt_l_MSTR [{shapeType=Capsule, length=1, radius=1}]
 dyn_skirt_r_01 [{shapeType=Capsule, length=6, radius=7.5, shapeName=dyn_skirt_r_01_capsule, center=<5.3667 0 7.2>, rotation=<rot 0 0 90>}]
 dyn_skirt_r_02 [{shapeType=Capsule, length=15, radius=6, shapeName=dyn_skirt_r_02_capsule, center=<5.3137 0 5.3>, rotation=<rot 0 0 90>}]
 dyn_skirt_r_03 [{shapeType=Capsule, length=15, radius=6.5, shapeName=dyn_skirt_r_03_capsule, center=<4.512 0 4.2>, rotation=<rot 0 0 90>}]
 dyn_skirt_r_MSTR [{shapeType=Capsule, length=1, radius=1}]
 pelvis [{shapeType=Capsule, length=8.4806, radius=10.1788, center=<-4.1132 -0.8885 -0.0007>}]
 spine_01 [{shapeType=Capsule, length=7.7754, radius=7.5232, center=<-3.5432 -0.0636 -0.0007>}]
 thigh_l [{shapeType=Capsule, length=35.6533, radius=9.7032, center=<-25.5462 -0.8392 -0.1542>, rotation=<rot 34.8586 93.1484 96.5491>}]
 thigh_r [{shapeType=Capsule, length=27.0468, radius=9.3282, center=<21.3088 -0.2106 -0.9794>, rotation=<rot 86.5422 179.6205 178.886>}]
}
constraints [
 {bone1Name=dyn_skirt_bk_l_02, bone2Name=dyn_skirt_bk_l_01, swing1LimitDegrees=45, swing2LimitDegrees=45, twistLimitDegrees=5}
 {bone1Name=dyn_skirt_bk_l_03, bone2Name=dyn_skirt_bk_l_02, swing1LimitDegrees=45, swing2LimitDegrees=45, twistLimitDegrees=5}
 {bone1Name=dyn_skirt_bk_02, bone2Name=dyn_skirt_bk_01, swing1LimitDegrees=45, swing2LimitDegrees=45, twistLimitDegrees=5}
 {bone1Name=dyn_skirt_bk_03, bone2Name=dyn_skirt_bk_02, swing1LimitDegrees=45, swing2LimitDegrees=45, twistLimitDegrees=5}
 {bone1Name=dyn_skirt_fr_mid_l_01, bone2Name=dyn_skirt_fr_mid_l_MSTR, swing1LimitDegrees=20, swing2LimitDegrees=45, twistLimitDegrees=5}
 {bone1Name=dyn_skirt_fr_mid_l_02, bone2Name=dyn_skirt_fr_mid_l_01, swing1LimitDegrees=20, swing2LimitDegrees=45, twistLimitDegrees=5}
 {bone1Name=dyn_skirt_fr_mid_l_03, bone2Name=dyn_skirt_fr_mid_l_02, swing1LimitDegrees=20, swing2LimitDegrees=45, twistLimitDegrees=5}
 {bone1Name=dyn_skirt_fr_l_01, bone2Name=dyn_skirt_fr_MSTR, swing1LimitDegrees=5, swing2LimitDegrees=45, twistLimitDegrees=5}
 {bone1Name=dyn_skirt_fr_l_02, bone2Name=dyn_skirt_fr_l_01, swing1LimitDegrees=5, swing2LimitDegrees=45, twistLimitDegrees=5}
 {bone1Name=dyn_skirt_fr_l_03, bone2Name=dyn_skirt_fr_l_02, swing1LimitDegrees=5, swing2LimitDegrees=45, twistLimitDegrees=5}
 {bone1Name=dyn_skirt_l_01, bone2Name=dyn_skirt_l_MSTR, swing1LimitDegrees=45, swing2LimitDegrees=45, twistLimitDegrees=5}
 {bone1Name=dyn_skirt_l_02, bone2Name=dyn_skirt_l_01, swing1LimitDegrees=45, swing2LimitDegrees=45, twistLimitDegrees=5}
 {bone1Name=dyn_skirt_l_03, bone2Name=dyn_skirt_l_02, swing1LimitDegrees=45, swing2LimitDegrees=45, twistLimitDegrees=5}
 {
  bone1Name=dyn_skirt_fr_mid_r_01
  bone2Name=dyn_skirt_fr_mid_r_MSTR
  swing1LimitDegrees=3
  swing1Motion=Locked
  swing2LimitDegrees=45
  twistLimitDegrees=5
 }
 {
  bone1Name=dyn_skirt_fr_mid_r_02
  bone2Name=dyn_skirt_fr_mid_r_01
  swing1LimitDegrees=3
  swing1Motion=Locked
  swing2LimitDegrees=45
  twistLimitDegrees=5
 }
 {
  bone1Name=dyn_skirt_fr_mid_r_03
  bone2Name=dyn_skirt_fr_mid_r_02
  swing1LimitDegrees=3
  swing1Motion=Locked
  swing2LimitDegrees=45
  twistLimitDegrees=5
 }
 {bone1Name=dyn_skirt_r_01, bone2Name=dyn_skirt_r_MSTR, swing1LimitDegrees=45, swing2LimitDegrees=45, twistLimitDegrees=5}
 {bone1Name=dyn_skirt_r_02, bone2Name=dyn_skirt_r_01, swing1LimitDegrees=45, swing2LimitDegrees=45, twistLimitDegrees=5}
 {bone1Name=dyn_skirt_r_03, bone2Name=dyn_skirt_r_02, swing1LimitDegrees=45, swing2LimitDegrees=45, twistLimitDegrees=5}
 {bone1Name=dyn_skirt_bk_r_02, bone2Name=dyn_skirt_bk_r_01, swing1LimitDegrees=45, swing2LimitDegrees=45, twistLimitDegrees=5}
 {bone1Name=dyn_skirt_bk_r_03, bone2Name=dyn_skirt_bk_r_02, swing1LimitDegrees=45, swing2LimitDegrees=45, twistLimitDegrees=5}
 {bone1Name=dyn_skirt_bk_l_01, bone2Name=pelvis, swing1LimitDegrees=45, swing2LimitDegrees=45, twistLimitDegrees=5}
 {bone1Name=dyn_skirt_bk_01, bone2Name=pelvis, swing1LimitDegrees=45, swing2LimitDegrees=45, twistLimitDegrees=5}
 {
  bone1Name=dyn_skirt_fr_mid_l_MSTR
  bone2Name=pelvis
  swing1LimitDegrees=45
  swing1Motion=Free
  swing2LimitDegrees=45
  swing2Motion=Free
  twistLimitDegrees=45
  twistMotion=Free
 }
 {
  bone1Name=dyn_skirt_fr_MSTR
  bone2Name=pelvis
  swing1LimitDegrees=45
  swing1Motion=Free
  swing2LimitDegrees=45
  swing2Motion=Free
  twistLimitDegrees=45
  twistMotion=Free
 }
 {
  bone1Name=dyn_skirt_l_MSTR
  bone2Name=pelvis
  swing1LimitDegrees=45
  swing1Motion=Free
  swing2LimitDegrees=45
  swing2Motion=Free
  twistLimitDegrees=45
  twistMotion=Free
 }
 {
  bone1Name=dyn_skirt_fr_mid_r_MSTR
  bone2Name=pelvis
  swing1LimitDegrees=45
  swing1Motion=Free
  swing2LimitDegrees=45
  swing2Motion=Free
  twistLimitDegrees=45
  twistMotion=Free
 }
 {
  bone1Name=dyn_skirt_r_MSTR
  bone2Name=pelvis
  swing1LimitDegrees=45
  swing1Motion=Free
  swing2LimitDegrees=45
  swing2Motion=Free
  twistLimitDegrees=45
  twistMotion=Free
 }
 {bone1Name=dyn_skirt_bk_r_01, bone2Name=pelvis, swing1LimitDegrees=45, swing2LimitDegrees=45, twistLimitDegrees=5}
 {
  bone1Name=thigh_l
  bone2Name=pelvis
  swing1LimitDegrees=45
  swing1Motion=Free
  swing2LimitDegrees=45
  swing2Motion=Free
  twistLimitDegrees=45
  twistMotion=Free
 }
 {
  bone1Name=thigh_r
  bone2Name=pelvis
  swing1LimitDegrees=45
  swing1Motion=Free
  swing2LimitDegrees=45
  swing2Motion=Free
  twistLimitDegrees=45
  twistMotion=Free
 }
 {
  bone1Name=spine_01
  bone2Name=pelvis
  swing1LimitDegrees=45
  swing1Motion=Free
  swing2LimitDegrees=45
  swing2Motion=Free
  twistLimitDegrees=45
  twistMotion=Free
 }
 {
  bone1Name=dyn_skirt_l_03
  bone2Name=dyn_skirt_fr_mid_l_03
  swing1LimitDegrees=45
  swing1Motion=Free
  swing2LimitDegrees=45
  swing2Motion=Free
  twistLimitDegrees=45
  twistMotion=Free
 }
 {
  bone1Name=dyn_skirt_r_03
  bone2Name=dyn_skirt_fr_mid_r_03
  swing1LimitDegrees=45
  swing1Motion=Free
  swing2LimitDegrees=45
  swing2Motion=Free
  twistLimitDegrees=45
  twistMotion=Free
 }
 {
  bone1Name=dyn_skirt_fr_mid_r_03
  bone2Name=dyn_skirt_fr_l_03
  swing1LimitDegrees=45
  swing1Motion=Free
  swing2LimitDegrees=45
  swing2Motion=Free
  twistLimitDegrees=45
  twistMotion=Free
 }
 {
  bone1Name=dyn_skirt_r_03
  bone2Name=dyn_skirt_bk_r_03
  swing1LimitDegrees=45
  swing1Motion=Free
  swing2LimitDegrees=45
  swing2Motion=Free
  twistLimitDegrees=45
  twistMotion=Free
 }
 {
  bone1Name=dyn_skirt_l_03
  bone2Name=dyn_skirt_bk_l_03
  swing1LimitDegrees=45
  swing1Motion=Free
  swing2LimitDegrees=45
  swing2Motion=Free
  twistLimitDegrees=45
  twistMotion=Free
 }
 {
  bone1Name=dyn_skirt_fr_seam_l_MSTR
  bone2Name=pelvis
  swing1LimitDegrees=45
  swing1Motion=Free
  swing2LimitDegrees=45
  swing2Motion=Free
  twistLimitDegrees=45
  twistMotion=Free
 }
 {bone1Name=dyn_skirt_fr_seam_l_01, bone2Name=dyn_skirt_fr_seam_l_MSTR, swing1LimitDegrees=5, swing2LimitDegrees=25, twistLimitDegrees=5}
 {bone1Name=dyn_skirt_fr_seam_l_02, bone2Name=dyn_skirt_fr_seam_l_01, swing1LimitDegrees=15, swing2LimitDegrees=45, twistLimitDegrees=5}
 {bone1Name=dyn_skirt_fr_seam_l_03, bone2Name=dyn_skirt_fr_seam_l_02, swing1LimitDegrees=15, swing2LimitDegrees=45, twistLimitDegrees=5}
 {
  bone1Name=dyn_skirt_fr_l_03
  bone2Name=dyn_skirt_fr_seam_l_03
  swing1LimitDegrees=45
  swing1Motion=Free
  swing2LimitDegrees=45
  swing2Motion=Free
  twistLimitDegrees=45
  twistMotion=Free
 }
 {
  bone1Name=dyn_skirt_fr_seam_l_01
  bone2Name=dyn_skirt_fr_mid_l_01
  swing1LimitDegrees=45
  swing1Motion=Locked
  swing2LimitDegrees=45
  swing2Motion=Locked
  twistLimitDegrees=45
  twistMotion=Locked
 }
]
