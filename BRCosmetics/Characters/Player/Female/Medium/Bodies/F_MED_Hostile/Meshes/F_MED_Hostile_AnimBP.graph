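# Post-process-style cosmetic AnimBP for the F_MED_Hostile female body, animating its skirt bones over the pose copied from the attached parent mesh.
# Thigh angles, crouch, sprint and jump alphas drive additive bone correctives before RigidBody simulation, whose physics asset switches for long backbling (cape).
# Further constraints and offsets shape the skirt, with extra correctives when crouched with physics disabled.
# The EventGraph computes those alphas and overrides gravity for swim-sprinting and skydiving.
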AnimBlueprint /BRCosmetics/Characters/Player/Female/Medium/Bodies/F_MED_Hostile/Meshes/F_MED_Hostile_AnimBP
target_skeleton=&/F_MED_Hostile_Skeleton
graphs [
 AnimGraph {nodes [
   n1 = RigidBody!(
    ComponentPose=Local To Component(LocalPose=Use cached pose 'initial Pose'())
    Alpha
    OverrideWorldGravity:Vector=Get gravity_override() -> (gravity_override:Vector)
    ExternalForce
   ) -> (Pose) {
    bClampLinearTranslationLimitToRefPose=true
    bOverrideWorldGravity=true
    overridePhysicsAsset=&/F_MED_Hostile_Dyn
    simulationSpace=BaseBoneSpace
    alphaBoolBlend {blendOption=Linear}
    baseBoneRef {boneName=spine_01}
    componentAppliedLinearAccClamp=<3000 3000 3000>
    componentLinearAccScale=<0.3 0.2 0.2>
    componentLinearVelScale=<0.3 0.2 0.2>
    overrideWorldGravity {z=-980}
   }
   Blend Poses by bool(BlendPose_0=Component To Local(ComponentPose=@n1), bActiveValue=Get bIsSkyDiving()) {blendType=Linear}
   n2 = "Constraint - dyn_skirt_fr_seam_l_01 (P)"(Alpha=1, ConstraintWeights_0=1) {
    alphaBoolBlend {blendOption=Linear}
    boneToModify {boneName=dyn_skirt_fr_seam_l_01}
    constraintSetup [{targetBone {boneName=dyn_skirt_fr_mid_l_01}}]
    constraintWeights [1]
   }
   "Constraint - dyn_skirt_fr_seam_l_02 (P)"(ComponentPose=@n2, Alpha=1, ConstraintWeights_0=1) {
    alphaBoolBlend {blendOption=Linear}
    boneToModify {boneName=dyn_skirt_fr_seam_l_02}
    constraintSetup [{targetBone {boneName=dyn_skirt_fr_mid_l_02}}]
    constraintWeights [1]
   }
   "Constraint - dyn_skirt_fr_mid_l_02 (P)"(Alpha=1, ConstraintWeights_0=1) {
    alphaBoolBlend {blendOption=Linear}
    boneToModify {boneName=dyn_skirt_fr_mid_l_02}
    constraintSetup [{targetBone {boneName=thigh_l}}]
    constraintWeights [1]
   }
   n3 = Get rigidBodyDisabled() -> (rigidBodyDisabled:double)
   n4 = "Transform (Modify) Bone - Bone: dyn_skirt_fr_seam_l_02"(Alpha=@n3, Translation="0,0,5", Rotation="15,0,0", Scale="1,1,1") {
    rotationMode=BMM_Additive
    rotationSpace=BCS_BoneSpace
    translationMode=BMM_Additive
    translationSpace=BCS_BoneSpace
    alphaBoolBlend {blendOption=Linear}
    boneToModify {boneName=dyn_skirt_fr_seam_l_02}
    rotation {pitch=15}
    translation {z=5}
   }
   "Transform (Modify) Bone - Bone: dyn_skirt_fr_mid_l_02"(ComponentPose=@n4, Alpha=@n3, Translation="0,0,-1", Rotation="0,0,0", Scale="1,1,1") {
    rotationMode=BMM_Additive
    rotationSpace=BCS_BoneSpace
    translationMode=BMM_Additive
    translationSpace=BCS_BoneSpace
    alphaBoolBlend {blendOption=Linear}
    boneToModify {boneName=dyn_skirt_fr_mid_l_02}
    translation {z=-1}
   }
   n5 = "Transform (Modify) Bone - Bone: dyn_skirt_fr_mid_l_01"(Alpha=1, Translation="0,0,-3", Rotation="-10,0,0", Scale="1,1,1") {
    rotationMode=BMM_Additive
    rotationSpace=BCS_BoneSpace
    translationMode=BMM_Additive
    translationSpace=BCS_BoneSpace
    alphaBoolBlend {blendOption=Linear}
    boneToModify {boneName=dyn_skirt_fr_mid_l_01}
    rotation {pitch=-10}
    translation {z=-3}
   }
   "Transform (Modify) Bone - Bone: dyn_skirt_fr_seam_l_01"(ComponentPose=@n5, Alpha=1, Translation="0,0,-3", Rotation="0,0,0", Scale="1,1,1") {
    rotationMode=BMM_Additive
    rotationSpace=BCS_BoneSpace
    translationMode=BMM_Additive
    translationSpace=BCS_BoneSpace
    alphaBoolBlend {blendOption=Linear}
    boneToModify {boneName=dyn_skirt_fr_seam_l_01}
    translation {z=-3}
   }
   "Transform (Modify) Bone - Bone: dyn_skirt_fr_seam_l_MSTR"(Alpha=1, Translation="0,0,3", Rotation="0,0,0", Scale="1,1,1") {
    rotationMode=BMM_Additive
    rotationSpace=BCS_BoneSpace
    translationMode=BMM_Additive
    translationSpace=BCS_BoneSpace
    alphaBoolBlend {blendOption=Linear}
    boneToModify {boneName=dyn_skirt_fr_seam_l_MSTR}
    translation {z=3}
   }
   "Transform (Modify) Bone - Bone: dyn_skirt_fr_seam_l_01"(Alpha=1, Translation="0,0,1", Rotation="60,0,0", Scale="1,1,1") {
    rotationMode=BMM_Additive
    rotationSpace=BCS_BoneSpace
    translationMode=BMM_Additive
    translationSpace=BCS_BoneSpace
    alphaBoolBlend {blendOption=Linear}
    boneToModify {boneName=dyn_skirt_fr_seam_l_01}
    rotation {pitch=60}
    translation {z=1}
   }
   RigidBody(OverrideWorldGravity="0,0,0") {
    bClampLinearTranslationLimitToRefPose=true
    bOverrideWorldGravity=true
    overridePhysicsAsset=&F_MED_Hostile_Dyn
    simulationSpace=BaseBoneSpace
    alphaBoolBlend {blendOption=Linear}
    baseBoneRef {boneName=pelvis}
    componentAppliedLinearAccClamp=<3000 3000 3000>
    componentLinearAccScale=<0.3 0.2 0.2>
    componentLinearVelScale=<0.3 0.2 0.2>
    overrideWorldGravity {z=-980}
   }
   n6 = Get thigh_rot_inr_r() -> (thigh_rot_inr_r:double)
   n7 = "Transform (Modify) Bone - Bone: dyn_skirt_r_01"(
    ComponentPose=Local To Component(LocalPose=Use cached pose 'With Disabled Rigidbodies'())
    Alpha=@n6
    Translation="0,0,1.5"
    Rotation="12,0,0"
    Scale="1,1,1"
   ) {
    rotationMode=BMM_Additive
    rotationSpace=BCS_BoneSpace
    translationMode=BMM_Additive
    translationSpace=BCS_BoneSpace
    alphaBoolBlend {blendInTime=0.1, blendOption=Linear, blendOutTime=0.1}
    boneToModify {boneName=dyn_skirt_r_01}
    rotation {pitch=12}
    translation {z=1.5}
   }
   n8 = "Transform (Modify) Bone - Bone: dyn_skirt_r_02"(ComponentPose=@n7, Alpha=@n6, Translation="0,0,1.5", Rotation="10,0,0", Scale="1,1,1") {
    rotationMode=BMM_Additive
    rotationSpace=BCS_BoneSpace
    translationMode=BMM_Additive
    translationSpace=BCS_BoneSpace
    alphaBoolBlend {blendInTime=0.1, blendOption=Linear, blendOutTime=0.1}
    boneToModify {boneName=dyn_skirt_r_02}
    rotation {pitch=10}
    translation {z=1.5}
   }
   n9 = Get thigh_rot_inr_l() -> (thigh_rot_inr_l:double)
   n10 = "Transform (Modify) Bone - Bone: dyn_skirt_bk_l_01"(ComponentPose=@n8, Alpha=@n9, Translation="0,0,0", Rotation="0,0,0", Scale="1,1,1") {
    rotationMode=BMM_Additive
    rotationSpace=BCS_BoneSpace
    translationMode=BMM_Additive
    translationSpace=BCS_BoneSpace
    alphaBoolBlend {blendInTime=0.1, blendOption=Linear, blendOutTime=0.1}
    boneToModify {boneName=dyn_skirt_bk_l_01}
   }
   n11 = "Transform (Modify) Bone - Bone: dyn_skirt_l_01"(ComponentPose=@n10, Alpha=@n9, Translation="0,0,0", Rotation="20,0,0", Scale="1,1,1") {
    rotationMode=BMM_Additive
    rotationSpace=BCS_BoneSpace
    translationMode=BMM_Additive
    translationSpace=BCS_BoneSpace
    alphaBoolBlend {blendInTime=0.1, blendOption=Linear, blendOutTime=0.1}
    boneToModify {boneName=dyn_skirt_l_01}
    rotation {pitch=20}
   }
   n12 = "Transform (Modify) Bone - Bone: dyn_skirt_fr_mid_l_01"(ComponentPose=@n11, Alpha=1, Translation="0,0,0", Rotation="0,0,0", Scale="1,1,1") {
    rotationMode=BMM_Additive
    rotationSpace=BCS_BoneSpace
    translationMode=BMM_Additive
    translationSpace=BCS_BoneSpace
    alphaBoolBlend {blendInTime=0.1, blendOption=Linear, blendOutTime=0.1}
    boneToModify {boneName=dyn_skirt_fr_mid_l_01}
   }
   n13 = Get thigh_rot_up_l() -> (thigh_rot_up_l:double)
   n14 = "Transform (Modify) Bone - Bone: dyn_skirt_fr_seam_l_01"(ComponentPose=@n12, Alpha=@n13, Translation="0,0,0", Rotation="2,0,0", Scale="1,1,1") {
    rotationMode=BMM_Additive
    rotationSpace=BCS_BoneSpace
    translationMode=BMM_Additive
    translationSpace=BCS_BoneSpace
    alphaBoolBlend {blendInTime=0.1, blendOption=Linear, blendOutTime=0.1}
    boneToModify {boneName=dyn_skirt_fr_seam_l_01}
    rotation {pitch=2}
   }
   n15 = "Transform (Modify) Bone - Bone: dyn_skirt_fr_seam_l_02"(ComponentPose=@n14, Alpha=@n13, Translation="0,0,0", Rotation="10,0,0", Scale="1,1,1") {
    rotationMode=BMM_Additive
    rotationSpace=BCS_BoneSpace
    translationMode=BMM_Additive
    translationSpace=BCS_BoneSpace
    alphaBoolBlend {blendInTime=0.1, blendOption=Linear, blendOutTime=0.1}
    boneToModify {boneName=dyn_skirt_fr_seam_l_02}
    rotation {pitch=10}
   }
   n16 = "Transform (Modify) Bone - Bone: dyn_skirt_fr_mid_l_01"(ComponentPose=@n15, Alpha=@n13, Translation="0,0,0", Rotation="-7,-13,30", Scale="1,1,1") {
    rotationMode=BMM_Additive
    rotationSpace=BCS_BoneSpace
    translationMode=BMM_Additive
    translationSpace=BCS_BoneSpace
    alphaBoolBlend {blendInTime=0.1, blendOption=Linear, blendOutTime=0.1}
    boneToModify {boneName=dyn_skirt_fr_mid_l_01}
    rotation=<rot -7 -13 30>
   }
   n17 = "Transform (Modify) Bone - Bone: dyn_skirt_fr_seam_l_03"(ComponentPose=@n16, Alpha=@n13, Translation="0,0,0", Rotation="-40,0,0", Scale="1,1,1") {
    rotationMode=BMM_Additive
    rotationSpace=BCS_BoneSpace
    translationMode=BMM_Additive
    translationSpace=BCS_BoneSpace
    alphaBoolBlend {blendInTime=0.1, blendOption=Linear, blendOutTime=0.1}
    boneToModify {boneName=dyn_skirt_fr_seam_l_03}
    rotation {pitch=-40}
   }
   n18 = "Transform (Modify) Bone - Bone: dyn_skirt_l_01"(ComponentPose=@n17, Alpha=@n13, Translation="0,0,0", Rotation="20,0,0", Scale="1,1,1") {
    rotationMode=BMM_Additive
    rotationSpace=BCS_BoneSpace
    translationMode=BMM_Additive
    translationSpace=BCS_BoneSpace
    alphaBoolBlend {blendInTime=0.1, blendOption=Linear, blendOutTime=0.1}
    boneToModify {boneName=dyn_skirt_l_01}
    rotation {pitch=20}
   }
   n19 = "Transform (Modify) Bone - Bone: dyn_skirt_l_02"(ComponentPose=@n18, Alpha=@n13, Translation="0,0,0", Rotation="-20,0,0", Scale="1,1,1") {
    rotationMode=BMM_Additive
    rotationSpace=BCS_BoneSpace
    translationMode=BMM_Additive
    translationSpace=BCS_BoneSpace
    alphaBoolBlend {blendInTime=0.1, blendOption=Linear, blendOutTime=0.1}
    boneToModify {boneName=dyn_skirt_l_02}
    rotation {pitch=-20}
   }
   n20 = "Transform (Modify) Bone - Bone: dyn_skirt_fr_l_01"!(ComponentPose=@n19, bAlphaBoolEnabled:bool=True, Translation="0,0,0", Rotation="10,0,0", Scale="1,1,1") -> (Pose) {
    alphaInputType=Bool
    rotationMode=BMM_Additive
    rotationSpace=BCS_BoneSpace
    translationMode=BMM_Additive
    translationSpace=BCS_BoneSpace
    alphaBoolBlend {blendInTime=0.1, blendOption=Linear, blendOutTime=0.1}
    boneToModify {boneName=dyn_skirt_fr_l_01}
    rotation {pitch=10}
   }
   n21 = "Transform (Modify) Bone - Bone: dyn_skirt_fr_l_02"!(ComponentPose=@n20, bAlphaBoolEnabled:bool=True, Translation="0,0,0", Rotation="10,0,0", Scale="1,1,1") -> (Pose) {
    alphaInputType=Bool
    rotationMode=BMM_Additive
    rotationSpace=BCS_BoneSpace
    translationMode=BMM_Additive
    translationSpace=BCS_BoneSpace
    alphaBoolBlend {blendInTime=0.1, blendOption=Linear, blendOutTime=0.1}
    boneToModify {boneName=dyn_skirt_fr_l_02}
    rotation {pitch=10}
   }
   n22 = "Transform (Modify) Bone - Bone: dyn_skirt_fr_seam_l_01"!(ComponentPose=@n21, bAlphaBoolEnabled:bool=True, Translation="0,0,0", Rotation="0,0,0", Scale="1,1,1") -> (Pose) {
    alphaInputType=Bool
    rotationMode=BMM_Additive
    rotationSpace=BCS_BoneSpace
    translationMode=BMM_Additive
    translationSpace=BCS_BoneSpace
    alphaBoolBlend {blendInTime=0.1, blendOption=Linear, blendOutTime=0.1}
    boneToModify {boneName=dyn_skirt_fr_seam_l_01}
   }
   n23 = "Constraint - dyn_skirt_bk_l_01 (P)"!(ComponentPose=@n22, bAlphaBoolEnabled:bool=True, ConstraintWeights_0=1) -> (Pose) {
    alphaInputType=Bool
    alphaBoolBlend {blendOption=Linear}
    boneToModify {boneName=dyn_skirt_bk_l_01}
    constraintSetup [{targetBone {boneName=spine_01}}]
    constraintWeights [1]
   }
   n24 = "Constraint - dyn_skirt_bk_01 (P)"!(ComponentPose=@n23, bAlphaBoolEnabled:bool=True, ConstraintWeights_0=1) -> (Pose) {
    alphaInputType=Bool
    alphaBoolBlend {blendOption=Linear}
    boneToModify {boneName=dyn_skirt_bk_01}
    constraintSetup [{targetBone {boneName=spine_01}}]
    constraintWeights [1]
   }
   n25 = "Constraint - dyn_skirt_bk_r_01 (P)"!(ComponentPose=@n24, bAlphaBoolEnabled:bool=True, ConstraintWeights_0=1) -> (Pose) {
    alphaInputType=Bool
    alphaBoolBlend {blendOption=Linear}
    boneToModify {boneName=dyn_skirt_bk_r_01}
    constraintSetup [{targetBone {boneName=spine_01}}]
    constraintWeights [1]
   }
   n26 = "Transform (Modify) Bone - Bone: dyn_skirt_bk_l_01"!(ComponentPose=@n25, bAlphaBoolEnabled:bool=True, Translation="0,0,0", Rotation="75,0,0", Scale="1,1,1") -> (Pose) {
    alphaInputType=Bool
    rotationMode=BMM_Additive
    rotationSpace=BCS_BoneSpace
    translationMode=BMM_Additive
    translationSpace=BCS_BoneSpace
    alphaBoolBlend {blendInTime=0.1, blendOption=Linear, blendOutTime=0.1}
    boneToModify {boneName=dyn_skirt_bk_l_01}
    rotation {pitch=75}
   }
   n27 = "Transform (Modify) Bone - Bone: dyn_skirt_bk_01"!(ComponentPose=@n26, bAlphaBoolEnabled:bool=True, Translation="0,0,0", Rotation="90,0,0", Scale="1,1,1") -> (Pose) {
    alphaInputType=Bool
    rotationMode=BMM_Additive
    rotationSpace=BCS_BoneSpace
    translationMode=BMM_Additive
    translationSpace=BCS_BoneSpace
    alphaBoolBlend {blendInTime=0.1, blendOption=Linear, blendOutTime=0.1}
    boneToModify {boneName=dyn_skirt_bk_01}
    rotation {pitch=90}
   }
   n28 = "Transform (Modify) Bone - Bone: dyn_skirt_bk_r_01"!(ComponentPose=@n27, bAlphaBoolEnabled:bool=True, Translation="0,0,0", Rotation="75,0,0", Scale="1,1,1") -> (Pose) {
    alphaInputType=Bool
    rotationMode=BMM_Additive
    rotationSpace=BCS_BoneSpace
    translationMode=BMM_Additive
    translationSpace=BCS_BoneSpace
    alphaBoolBlend {blendInTime=0.1, blendOption=Linear, blendOutTime=0.1}
    boneToModify {boneName=dyn_skirt_bk_r_01}
    rotation {pitch=75}
   }
   n29 = Component To Local(ComponentPose=@n28)
   n30 = Use cached pose 'With Disabled Rigidbodies'()
   n31 = AND Boolean(A=Get bIsCrouching(), B=NOT Boolean(A=Get bRigidBodyEnabled()))
   n32 = "Transform (Modify) Bone - Bone: dyn_skirt_fr_seam_l_02"!(
    ComponentPose=Local To Component(LocalPose=Blend Poses by bool(BlendPose_0=@n29, BlendPose_1=@n30, bActiveValue=@n31) {blendType=Linear})
    bAlphaBoolEnabled:bool=True
    Translation="0,0,3"
    Rotation="0,0,0"
    Scale="1,1,1"
   ) -> (Pose) {
    alphaInputType=Bool
    rotationMode=BMM_Additive
    rotationSpace=BCS_BoneSpace
    translationMode=BMM_Additive
    translationSpace=BCS_BoneSpace
    alphaBoolBlend {blendOption=Linear}
    boneToModify {boneName=dyn_skirt_fr_seam_l_02}
    translation {z=3}
   }
   n33 = "Transform (Modify) Bone - Bone: dyn_skirt_fr_mid_l_02"!(ComponentPose=@n32, bAlphaBoolEnabled:bool=True, Translation="0,0,-1", Rotation="0,0,0", Scale="1,1,1") -> (Pose) {
    alphaInputType=Bool
    rotationMode=BMM_Additive
    rotationSpace=BCS_BoneSpace
    translationMode=BMM_Additive
    translationSpace=BCS_BoneSpace
    alphaBoolBlend {blendOption=Linear}
    boneToModify {boneName=dyn_skirt_fr_mid_l_02}
    translation {z=-1}
   }
   n34 = "Transform (Modify) Bone - Bone: dyn_skirt_fr_seam_l_01"(ComponentPose=@n33, bAlphaBoolEnabled=True, Translation="0,0,1.3", Rotation="0,0,0", Scale="1,1,1") {
    alphaInputType=Bool
    rotationMode=BMM_Additive
    rotationSpace=BCS_BoneSpace
    translationMode=BMM_Additive
    translationSpace=BCS_BoneSpace
    alphaBoolBlend {blendOption=Linear}
    boneToModify {boneName=dyn_skirt_fr_seam_l_01}
    translation {z=1.3}
   }
   Output Pose(Result=Component To Local(ComponentPose=@n34)) {layerGroup=None}
   n35 = "Constraint - dyn_skirt_fr_seam_l_01 (P)"(ComponentPose=Local To Component(LocalPose=Copy Pose From Mesh() {bUseAttachedParent=true}), Alpha=1, ConstraintWeights_0=1) {
    alphaBoolBlend {blendOption=Linear}
    boneToModify {boneName=dyn_skirt_fr_seam_l_01}
    constraintSetup [{targetBone {boneName=dyn_skirt_fr_mid_l_01}}]
    constraintWeights [1]
   }
   n36 = "Constraint - dyn_skirt_fr_seam_l_02 (P)"(ComponentPose=@n35, Alpha=1, ConstraintWeights_0=1) {
    alphaBoolBlend {blendOption=Linear}
    boneToModify {boneName=dyn_skirt_fr_seam_l_02}
    constraintSetup [{targetBone {boneName=dyn_skirt_fr_mid_l_02}}]
    constraintWeights [1]
   }
   n37 = "Transform (Modify) Bone - Bone: dyn_skirt_fr_mid_l_MSTR"(
    ComponentPose=@n36
    Alpha=Get thigh_rotation_l() -> (thigh_rotation_l:double)
    Translation="-7.731,-1.109,3.397"
    Rotation="44.415,-65.435,-79.400"
    Scale="1,1,1"
   ) {
    rotationMode=BMM_Additive
    rotationSpace=BCS_BoneSpace
    translationMode=BMM_Additive
    translationSpace=BCS_BoneSpace
    alphaBoolBlend {blendOption=Linear}
    alphaScaleBiasClamp {interpSpeedIncreasing=40}
    boneToModify {boneName=dyn_skirt_fr_mid_l_MSTR}
    rotation=<rot 44.4149 -65.4352 -79.3998>
    translation=<-7.731 -1.109 3.397>
   }
   n38 = "Transform (Modify) Bone - Bone: dyn_skirt_fr_mid_l_MSTR"(
    ComponentPose=@n37
    Alpha=Get crouch sprinting() -> (crouch sprinting:double)
    Translation="0,0,0"
    Rotation="-35.432,-80.754,0.000"
    Scale="1,1,1"
   ) {
    rotationMode=BMM_Additive
    rotationSpace=BCS_BoneSpace
    translationMode=BMM_Additive
    translationSpace=BCS_BoneSpace
    alphaBoolBlend {blendOption=Linear}
    alphaScaleBiasClamp {interpSpeedIncreasing=40}
    boneToModify {boneName=dyn_skirt_fr_mid_l_MSTR}
    rotation {pitch=-35.432, yaw=-80.754}
   }
   n39 = Get crouch alpha() -> (crouch alpha:double)
   n40 = "Transform (Modify) Bone - Bone: dyn_skirt_fr_seam_l_MSTR"(ComponentPose=@n38, Alpha=@n39, Translation="0,5,0", Rotation="0,0,0", Scale="1,1,1") {
    rotationMode=BMM_Additive
    rotationSpace=BCS_BoneSpace
    translationMode=BMM_Additive
    translationSpace=BCS_BoneSpace
    alphaBoolBlend {blendOption=Linear}
    alphaScaleBiasClamp {bInterpResult=true, interpSpeedDecreasing=5, interpSpeedIncreasing=20}
    boneToModify {boneName=dyn_skirt_fr_seam_l_MSTR}
    translation {y=5}
   }
   n41 = "Transform (Modify) Bone - Bone: dyn_skirt_fr_MSTR"(ComponentPose=@n40, Alpha=@n39, Translation="0,0,0", Rotation="-15,15,0", Scale="1,1,1") {
    rotationMode=BMM_Additive
    rotationSpace=BCS_BoneSpace
    translationMode=BMM_Additive
    translationSpace=BCS_BoneSpace
    alphaBoolBlend {blendOption=Linear}
    alphaScaleBiasClamp {bInterpResult=true, interpSpeedDecreasing=5, interpSpeedIncreasing=20}
    boneToModify {boneName=dyn_skirt_fr_MSTR}
    rotation {pitch=-15, yaw=15}
   }
   n42 = "Transform (Modify) Bone - Bone: dyn_skirt_fr_mid_r_MSTR"(
    ComponentPose=@n41
    Alpha=Get thigh rotation_r() -> (thigh rotation_r:double)
    Translation="-6.552,0.000,0.000"
    Rotation="71.999,0.000,0.000"
    Scale="1,1,1"
   ) {
    rotationMode=BMM_Additive
    rotationSpace=BCS_BoneSpace
    translationMode=BMM_Additive
    translationSpace=BCS_BoneSpace
    alphaBoolBlend {blendOption=Linear}
    alphaScaleBiasClamp {bInterpResult=true, interpSpeedDecreasing=40, interpSpeedIncreasing=40}
    boneToModify {boneName=dyn_skirt_fr_mid_r_MSTR}
    rotation {pitch=71.9989}
    translation {x=-6.552}
   }
   n43 = "Transform (Modify) Bone - Bone: dyn_skirt_fr_mid_r_MSTR"(
    ComponentPose=@n42
    Alpha=Get jumping alpha() -> (jumping alpha:double)
    Translation="-8.620,-0.000,-0.000"
    Rotation="27.635,0.000,0.000"
    Scale="1,1,1"
   ) {
    rotationMode=BMM_Additive
    rotationSpace=BCS_BoneSpace
    translationMode=BMM_Additive
    translationSpace=BCS_BoneSpace
    alphaBoolBlend {blendOption=Linear}
    alphaScaleBiasClamp {bInterpResult=true, interpSpeedDecreasing=20, interpSpeedIncreasing=20}
    boneToModify {boneName=dyn_skirt_fr_mid_r_MSTR}
    rotation {pitch=27.635}
    translation {x=-8.62}
   }
   Save cached pose 'initial Pose'(Pose=Component To Local(ComponentPose=@n43)) {cacheName=initial Pose}
   n44 = Get gravity_override()
   n45 = RigidBody(ComponentPose=Local To Component(LocalPose=Use cached pose 'initial Pose'()), OverrideWorldGravity=@n44) {
    bClampLinearTranslationLimitToRefPose=true
    bOverrideWorldGravity=true
    overridePhysicsAsset=&/F_MED_Hostile_DynCapeNew
    simulationSpace=BaseBoneSpace
    alphaBoolBlend {blendOption=Linear}
    baseBoneRef {boneName=pelvis}
    componentAppliedLinearAccClamp=<3000 3000 3000>
    componentLinearAccScale=<0.3 0.2 0.2>
    componentLinearVelScale=<0.3 0.2 0.2>
    overrideWorldGravity {z=-980}
   }
   n46 = Component To Local(ComponentPose=@n45)
   n47 = RigidBody(ComponentPose=Local To Component(LocalPose=Use cached pose 'initial Pose'()), OverrideWorldGravity=@n44) {
    bClampLinearTranslationLimitToRefPose=true
    bOverrideWorldGravity=true
    overridePhysicsAsset=&F_MED_Hostile_Dyn
    simulationSpace=BaseBoneSpace
    alphaBoolBlend {blendOption=Linear}
    baseBoneRef {boneName=pelvis}
    componentAppliedLinearAccClamp=<3000 3000 3000>
    componentLinearAccScale=<0.3 0.2 0.2>
    componentLinearVelScale=<0.3 0.2 0.2>
    overrideWorldGravity {z=-980}
   }
   n48 = Component To Local(ComponentPose=@n47)
   n49 = Get isWearingCape() -> (isWearingCape:bool)
   n50 = Get rigidBodyDisabled()
   n51 = "Constraint - dyn_skirt_fr_mid_r_MSTR (R)"(
    ComponentPose=Local To Component(LocalPose=Blend Poses by bool(BlendPose_0=@n46, BlendPose_1=@n48, bActiveValue=@n49) {blendType=Linear})
    Alpha=@n50
    ConstraintWeights_0=1
   ) {
    alphaBoolBlend {blendOption=Linear}
    boneToModify {boneName=dyn_skirt_fr_mid_r_MSTR}
    constraintSetup [{transformType=Rotation, targetBone {boneName=thigh_r}}]
    constraintWeights [1]
   }
   n52 = "Constraint - dyn_skirt_fr_mid_l_MSTR (R)"(ComponentPose=@n51, Alpha=@n50, ConstraintWeights_0=1) {
    alphaBoolBlend {blendOption=Linear}
    boneToModify {boneName=dyn_skirt_fr_mid_l_MSTR}
    constraintSetup [{transformType=Rotation, targetBone {boneName=thigh_l}}]
    constraintWeights [1]
   }
   n53 = "Transform (Modify) Bone - Bone: dyn_skirt_fr_mid_l_01"!(ComponentPose=@n52, bAlphaBoolEnabled:bool=True, Translation="0,0,0", Rotation="2,-10,0", Scale="1,1,1") -> (Pose) {
    alphaInputType=Bool
    rotationMode=BMM_Additive
    rotationSpace=BCS_BoneSpace
    translationMode=BMM_Additive
    translationSpace=BCS_BoneSpace
    alphaBoolBlend {blendInTime=0.1, blendOption=Linear, blendOutTime=0.1}
    boneToModify {boneName=dyn_skirt_fr_mid_l_01}
    rotation {pitch=2, yaw=-10}
   }
   n54 = "Constraint - dyn_skirt_fr_seam_l_01 (PP)"!(ComponentPose=@n53, Alpha=1, ConstraintWeights_0=0.9, ConstraintWeights_1:float=0.1) -> (Pose) {
    alphaBoolBlend {blendInTime=0.1, blendOption=Linear, blendOutTime=0.1}
    boneToModify {boneName=dyn_skirt_fr_seam_l_01}
    constraintSetup [{targetBone {boneName=dyn_skirt_fr_mid_l_MSTR}}, {targetBone {boneName=dyn_skirt_fr_l_01}}]
    constraintWeights [1,1]
   }
   n55 = "Transform (Modify) Bone - Bone: dyn_skirt_fr_seam_l_02"(ComponentPose=@n54, bAlphaBoolEnabled=True, Translation="0,0,1.5", Rotation="5,0,0", Scale="1,1,1") {
    alphaInputType=Bool
    rotationMode=BMM_Additive
    rotationSpace=BCS_BoneSpace
    translationMode=BMM_Additive
    translationSpace=BCS_BoneSpace
    alphaBoolBlend {blendInTime=0.1, blendOption=Linear, blendOutTime=0.1}
    boneToModify {boneName=dyn_skirt_fr_seam_l_02}
    rotation {pitch=5}
    translation {z=1.5}
   }
   n56 = "Transform (Modify) Bone - Bone: dyn_skirt_fr_seam_l_03"!(ComponentPose=@n55, bAlphaBoolEnabled:bool=True, Translation="0,0,0", Rotation="-15,0,0", Scale="1,1,1") -> (Pose) {
    alphaInputType=Bool
    rotationMode=BMM_Additive
    rotationSpace=BCS_BoneSpace
    translationMode=BMM_Additive
    translationSpace=BCS_BoneSpace
    alphaBoolBlend {blendInTime=0.1, blendOption=Linear, blendOutTime=0.1}
    boneToModify {boneName=dyn_skirt_fr_seam_l_03}
    rotation {pitch=-15}
   }
   n57 = "Transform (Modify) Bone - Bone: dyn_skirt_fr_mid_l_01"(ComponentPose=@n56, bAlphaBoolEnabled=True, Translation="0,0,0", Rotation="1,0,10", Scale="1,1,1") {
    alphaInputType=Bool
    rotationMode=BMM_Additive
    rotationSpace=BCS_BoneSpace
    translationMode=BMM_Additive
    translationSpace=BCS_BoneSpace
    alphaBoolBlend {blendInTime=0.1, blendOption=Linear, blendOutTime=0.1}
    boneToModify {boneName=dyn_skirt_fr_mid_l_01}
    rotation {pitch=1, roll=10}
   }
   n58 = "Transform (Modify) Bone - Bone: dyn_skirt_l_01"!(ComponentPose=@n57, bAlphaBoolEnabled:bool=True, Translation="0,0,1.5", Rotation="4,5,0", Scale="1,1,1") -> (Pose) {
    alphaInputType=Bool
    rotationMode=BMM_Additive
    rotationSpace=BCS_BoneSpace
    translationMode=BMM_Additive
    translationSpace=BCS_BoneSpace
    alphaBoolBlend {blendInTime=0.1, blendOption=Linear, blendOutTime=0.1}
    boneToModify {boneName=dyn_skirt_l_01}
    rotation {pitch=4, yaw=5}
    translation {z=1.5}
   }
   n59 = "Transform (Modify) Bone - Bone: dyn_skirt_r_01"!(ComponentPose=@n58, bAlphaBoolEnabled:bool=True, Translation="0,0,1.5", Rotation="5,0,0", Scale="1,1,1") -> (Pose) {
    alphaInputType=Bool
    rotationMode=BMM_Additive
    rotationSpace=BCS_BoneSpace
    translationMode=BMM_Additive
    translationSpace=BCS_BoneSpace
    alphaBoolBlend {blendInTime=0.1, blendOption=Linear, blendOutTime=0.1}
    boneToModify {boneName=dyn_skirt_r_01}
    rotation {pitch=5}
    translation {z=1.5}
   }
   n60 = "Transform (Modify) Bone - Bone: dyn_skirt_fr_mid_r_01"!(ComponentPose=@n59, bAlphaBoolEnabled:bool=True, Translation="0,0,3", Rotation="3.5,5,-15", Scale="1,1,1") -> (Pose) {
    alphaInputType=Bool
    rotationMode=BMM_Additive
    rotationSpace=BCS_BoneSpace
    translationMode=BMM_Additive
    translationSpace=BCS_BoneSpace
    alphaBoolBlend {blendInTime=0.1, blendOption=Linear, blendOutTime=0.1}
    boneToModify {boneName=dyn_skirt_fr_mid_r_01}
    rotation=<rot 3.5 5 -15>
    translation {z=3}
   }
   n61 = "Transform (Modify) Bone - Bone: dyn_skirt_fr_mid_r_02"!(ComponentPose=@n60, bAlphaBoolEnabled:bool=True, Translation="0,0,3", Rotation="-20,0,0", Scale="1,1,1") -> (Pose) {
    alphaInputType=Bool
    rotationMode=BMM_Additive
    rotationSpace=BCS_BoneSpace
    translationMode=BMM_Additive
    translationSpace=BCS_BoneSpace
    alphaBoolBlend {blendInTime=0.1, blendOption=Linear, blendOutTime=0.1}
    boneToModify {boneName=dyn_skirt_fr_mid_r_02}
    rotation {pitch=-20}
    translation {z=3}
   }
   n62 = "Constraint - dyn_skirt_fr_l_01 (PP)"!(ComponentPose=@n61, Alpha=1, ConstraintWeights_0=0.5, ConstraintWeights_1:float=0.5) -> (Pose) {
    alphaBoolBlend {blendInTime=0.1, blendOption=Linear, blendOutTime=0.1}
    boneToModify {boneName=dyn_skirt_fr_l_01}
    constraintSetup [{targetBone {boneName=dyn_skirt_fr_seam_l_01}}, {targetBone {boneName=dyn_skirt_fr_mid_r_01}}]
    constraintWeights [1,1]
   }
   n63 = "Transform (Modify) Bone - Bone: dyn_skirt_fr_l_01"(ComponentPose=@n62, bAlphaBoolEnabled=True, Translation="0,0,1", Rotation="0,0,0", Scale="1,1,1") {
    alphaInputType=Bool
    rotationMode=BMM_Additive
    rotationSpace=BCS_BoneSpace
    translationMode=BMM_Additive
    translationSpace=BCS_BoneSpace
    alphaBoolBlend {blendInTime=0.1, blendOption=Linear, blendOutTime=0.1}
    boneToModify {boneName=dyn_skirt_fr_l_01}
    translation {z=1}
   }
   Save cached pose 'With Disabled Rigidbodies'(Pose=Component To Local(ComponentPose=@n63)) {cacheName=With Disabled Rigidbodies}
   Get bIsInFrontEnd()
  ]}
 EventGraph {nodes [
   Event BlueprintUpdateAnimation()
   > n1 = Sequence!(execute) -> (then_0, then_1, then_2:Exec, then_3:Exec, then_4:Exec)
   > n2 = Branch(Condition=Get bIsSwimSprinting() -> (bIsSwimSprinting:bool))
   > Set gravity_override(gravity_override="0, 0, 0")
   n3 = Branch(execute=@n2.else, Condition=Get bIsSkyDiving())
   > n4 = Branch(Condition=Get bIsParachuteOpened())
   Set gravity_override(execute=[@n3.else, @n4.then], gravity_override="0, 0,-980")
   n5 = Branch(execute=@n4.else, Condition="float > float"(A=Get PawnVelocityZ(), B=50) {is=title K2Node_CallFunction})
   > Set gravity_override(gravity_override="0, 0,200")
   Set gravity_override(execute=@n5.else, gravity_override="0, 0,-200")
   n6 = BreakTransform(InTransform=GetDeltaTransformFromRefPose(self=GetOwnerSkeletalMeshComponent(), BoneName=thigh_r, BaseName=pelvis))
   Set thigh rotation_r(execute=@n1.then_1, thigh rotation_r:double=MapRangeClamped(Value=BreakRotator(InRot=@n6.Rotation).Yaw, InRangeA=-50, InRangeB=-90)) -> (Output_Get:double)
   n7 = BreakTransform(InTransform=GetDeltaTransformFromRefPose(self=GetOwnerSkeletalMeshComponent(), BoneName=thigh_l, BaseName=pelvis))
   > Set thigh_rotation_l(thigh_rotation_l:double=MapRangeClamped(Value=BreakRotator(InRot=@n7.Rotation).Yaw, InRangeA=-50, InRangeB=-90)) -> (Output_Get:double)
   > n8 = Branch(Condition=Get bIsJumping())
   Set jumping alpha(execute=[@n9.then, @n8.then], jumping alpha:double=1) -> (Output_Get:double)
   n9 = Branch(execute=@n8.else, Condition=Get bIsFalling())
   Set jumping alpha(execute=@n9.else, jumping alpha=0)
   n10 = Branch(execute=@n1.then_2, Condition=Get bIsCrouchSprinting())
   > Set crouch sprinting(crouch sprinting:double=1) -> (Output_Get:double)
   Set crouch sprinting(execute=@n10.else, crouch sprinting=0)
   n11 = Branch(execute=@n1.then_3, Condition=Get bRigidBodyEnabled())
   > Set rigidBodyDisabled(rigidBodyDisabled:double=0) -> (Output_Get:double)
   Set rigidBodyDisabled(execute=@n11.else, rigidBodyDisabled=1)
   n12 = GetOwnerSkeletalMeshComponent()
   n13 = BreakRotator(InRot=BreakTransform(InTransform=GetDeltaTransformFromRefPose(self=@n12, BoneName=thigh_r, BaseName=pelvis)).Rotation)
   Set thigh_rot_up_r(execute=@n1.then_4, thigh_rot_up_r:double=MapRangeClamped(Value=@n13.Yaw, InRangeA=-15, InRangeB=-100)) -> (Output_Get:double)
   > Set thigh_rot_inr_r(thigh_rot_inr_r:double=MapRangeClamped(Value=@n13.Pitch, InRangeA=5, InRangeB=-15, OutRangeA=1, OutRangeB=0)) -> (Output_Get:double)
   > Set thigh_rot_otr_r(thigh_rot_otr_r:double=MapRangeClamped(Value=@n13.Pitch, InRangeA=20, InRangeB=50)) -> (Output_Get:double)
   n14 = GetOwnerSkeletalMeshComponent()
   n15 = BreakRotator(InRot=BreakTransform(InTransform=GetDeltaTransformFromRefPose(self=@n14, BoneName=thigh_l, BaseName=pelvis)).Rotation)
   > Set thigh_rot_up_l(thigh_rot_up_l:double=MapRangeClamped(Value=@n15.Yaw, InRangeA=-10, InRangeB=-100)) -> (Output_Get:double)
   > Set thigh_rot_inr_l(thigh_rot_inr_l:double=MapRangeClamped(Value=@n15.Pitch, InRangeA=0, InRangeB=-10, OutRangeA=1, OutRangeB=0)) -> (Output_Get:double)
   > n16 = Set thigh_rot_otr_l(thigh_rot_otr_l:double=MapRangeClamped(Value=@n15.Pitch, InRangeA=20, InRangeB=50)) -> (Output_Get:double)
   Set head Rot(
    execute=[@n20.OutputPin, @n16.then]
    head Rot:double=GetAlphaFromReferenceTransformDelta(SkeletalMeshComponent=@n14, BoneName=upperarm_l, BaseName=spine_05, DeltaComponentType=EulerY, InputRange=(X=0,Y=40), TwistAxis="1,0,0")
   ) -> (Output_Get:double)
   n17 = Get bRigidBodyEnabled()
   n18 = Get bIsCrouching()
   > Set RigidOffStand(RigidOffStand:bool=AND Boolean(A=NOT Boolean(A=@n17), B=NOT Boolean(A=@n18))) -> (Output_Get:bool)
   Event OnBeginCrouching()
   > Set crouch alpha(crouch alpha:double=1) -> (Output_Get:double)
   Event OnEndCrouching()
   > Set crouch alpha(crouch alpha=0)
   Event OnCharacterCustomizationLoaded()
   n19 = CheckForBackbling()
   > Set isWearingBackbling(isWearingBackbling=@n19)
   > Set isWearingCape(isWearingCape=@n19.bIsBackblingLong)
   n20 = Reroute Node()
   PrintString!(execute, InString="To String (Float)"(InDouble=Get PawnSpeed2D()), bPrintToScreen, bPrintToLog, TextColor, Duration, Key) -> (then)
   Get PawnMovementDirectionAzimuth() -> (PawnMovementDirectionAzimuth:float)
  ]}
]
variables [
 crouch alpha {type=number}
 crouch sprinting {type=number}
 gravity_override {type=Vector}
 head Rot {type=number}
 isWearingBackbling {type=boolean}
 isWearingCape {type=boolean}
 jumping alpha {type=number}
 rigidBodyDisabled {type=number}
 RigidOffStand {type=boolean}
 thigh rotation_r {type=number}
 thigh_rot_inr_l {type=number}
 thigh_rot_inr_r {type=number}
 thigh_rot_otr_l {type=number}
 thigh_rot_otr_r {type=number}
 thigh_rot_up_l {type=number}
 thigh_rot_up_r {type=number}
 thigh_rotation_l {type=number}
]
