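# Physics asset of collision shapes for the F_MED_Hostile female cosmetic body, named as colliders for its cape.
# Capsules cover the head, clavicles, spine, pelvis and feet, with paired capsules on the middle spine bones.
# Large capsules also ride the back and side dyn_skirt third bones, adding collision volume around the skirt.
# Arm, leg and neck bodies have no shapes, and the asset has no constraints.
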PhysicsAsset /BRCosmetics/Characters/Player/Female/Medium/Bodies/F_MED_Hostile/Meshes/F_MED_Hostile_CapeColliders
bodies {
 calf_l []
 calf_r []
 clavicle_l [{shapeType=Capsule, length=7.6964, radius=8.6353, center=<7.5002 -0.9794 -1.0685>, rotation=<rot -77.5566 177.3915 7.7314>}]
 clavicle_r [{shapeType=Capsule, length=7.6964, radius=8.2957, center=<-7.4905 0.979 1.1123>, rotation=<rot -77.5565 177.3915 -172.2685>}]
 dyn_skirt_bk_03 [{shapeType=Capsule, length=15, radius=12.5, shapeName=dyn_skirt_bk_03_capsule, center=<-2.7858 0 -7.76>, rotation=<rot 90 -180 180>}]
 dyn_skirt_bk_l_03 [{
   shapeType=Capsule
   length=15
   radius=12.5
   shapeName=dyn_skirt_bk_l_03_capsule
   center=<-3.5643 4.4154 -8.7615>
   rotation=<rot 90 -179.1048 170.8951>
  }]
 dyn_skirt_bk_r_03 [{
   shapeType=Capsule
   length=15
   radius=12.5
   shapeName=dyn_skirt_bk_r_03_capsule
   center=<-3.6451 -4.8736 -8.897>
   rotation=<rot 90 14.0362 24.0364>
  }]
 dyn_skirt_l_03 [{
   shapeType=Capsule
   length=12
   radius=13
   shapeName=dyn_skirt_l_03_capsule
   center=<-2.4939 0.7483 -5.3586>
   rotation=<rot 90 -84.0088 -84.0094>
  }]
 dyn_skirt_r_03 [{
   shapeType=Capsule
   length=12
   radius=13
   shapeName=dyn_skirt_r_03_capsule
   center=<-2.4939 -1.9643 -5.8862>
   rotation=<rot 90 84.0088 84.0093>
  }]
 foot_l [{shapeType=Capsule, length=12.5982, radius=6.5127, center=<-2.2072 6.0534 -0.3367>, rotation=<rot 0.0388 1.4701 -89.3982>}]
 foot_r [{shapeType=Capsule, length=12.5982, radius=6.5127, center=<2.2072 -6.0534 0.3367>, rotation=<rot 0.0388 1.4701 90.6018>}]
 head [{shapeType=Capsule, length=0, radius=18, center=<9.0697 0.0746 0>, rotation=<rot 0 -90 0>}]
 lowerarm_l []
 lowerarm_r []
 neck_02 []
 pelvis [{shapeType=Capsule, length=10.4028, radius=11.9096, center=<-1.9544 -3.8764 -0.1981>}]
 spine_01 [{shapeType=Capsule, length=9.0171, radius=11.4854, center=<-2.8183 -1.0873 0>}]
 spine_02 [
  {shapeType=Capsule, length=6.2984, radius=8.9854, center=<1.179 0.9622 0>}
  {shapeType=Capsule, length=20.6875, radius=9.3125, center=<-1.8317 2.606 0>, rotation=<rot 90 0 0>}
 ]
 spine_03 [
  {shapeType=Capsule, length=3.2984, radius=10.4854, center=<1.1242 2.3725 0>}
  {shapeType=Capsule, length=20.7188, radius=9.2812, center=<-1.2326 2.6059 0>, rotation=<rot 90 0 0>}
 ]
 spine_04 [
  {shapeType=Capsule, length=5.8296, radius=11.6416, center=<5.5182 1.4112 0>}
  {shapeType=Capsule, length=10.0312, radius=9.2812, center=<-1.2326 1.308 0>, rotation=<rot 90 0 0>}
 ]
 spine_05 [{shapeType=Capsule, length=10.9859, radius=10.0791, center=<5.6377 0.3227 0>}]
 thigh_l []
 thigh_r []
 upperarm_l []
 upperarm_r []
}
constraints []
