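# Physics asset for the face accessory mesh of the medium female Hostile player body, setting up dynamic hair, bangs and hat bones.
# Capsule bodies on head, neck, upper spine and clavicles form the body volumes around the dynamic bones.
# Chains of capsules for front, back and side hair, bangs and a single hat bone hang from the head through angular-limited constraints.
# Bangs are nearly rigid, while hair and hat swing more freely.
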PhysicsAsset /BRCosmetics/Characters/Player/Female/Medium/Bodies/F_MED_Hostile/Meshes/Parts/F_MED_Hostile_FaceAcc_Dyn
bodies {
 clavicle_l [{shapeType=Capsule, length=11.3047, radius=7.2035, center=<5.1748 -0.7569 -1.363>, rotation=<rot 87.73 0 -2.8936>}]
 clavicle_r [{shapeType=Capsule, length=10.6135, radius=7.8948, center=<-5.1748 0.7569 1.363>, rotation=<rot 87.73 0 -2.8936>}]
 dyn_bangs_l_01 [{shapeType=Capsule, length=3.284, radius=0.716, center=<2.2 0 0>, rotation=<rot 90 0 0>}]
 dyn_bangs_l_02 [{shapeType=Capsule, length=3.4211, radius=0.5789, center=<2.2 0 0>, rotation=<rot 90 0 0>}]
 dyn_bangs_r_01 [{shapeType=Capsule, length=2.9361, radius=1.0639, center=<2.2 0 0>, rotation=<rot 90 0 0>}]
 dyn_bangs_r_02 [{shapeType=Capsule, length=2.9361, radius=1.0639, center=<2.2 0 0>, rotation=<rot 90 0 0>}]
 dyn_bangs_r_03 [{shapeType=Capsule, length=0.9506, radius=1.0639, center=<1.6362 0 0>, rotation=<rot 90 0 0>}]
 dyn_hair_bk_01 [{shapeType=Capsule, length=1.3, radius=1.5, center=<1 0 0>, rotation=<rot 90 0 0>}]
 dyn_hair_bk_02 [{shapeType=Capsule, length=1.3, radius=1.5, center=<1.5 0 0>, rotation=<rot 90 0 0>}]
 dyn_hair_bk_03 [{shapeType=Capsule, length=0.9, radius=1.5, center=<2.2 0 0>, rotation=<rot 90 0 0>}]
 dyn_hair_bk_l_01 [{shapeType=Capsule, length=1.3, radius=1.5, center=<1 0 0>, rotation=<rot 90 0 0>}]
 dyn_hair_bk_l_02 [{shapeType=Capsule, length=1.3, radius=1.5, center=<2.2 0 0>, rotation=<rot 90 0 0>}]
 dyn_hair_bk_l_03 [{shapeType=Capsule, length=0.9, radius=1, center=<1.6362 0 0>, rotation=<rot 90 0 0>}]
 dyn_hair_bk_r_01 [{shapeType=Capsule, length=1.3, radius=1.5, center=<1 0 0>, rotation=<rot 90 0 0>}]
 dyn_hair_bk_r_02 [{shapeType=Capsule, length=1.3, radius=1.5, center=<2.2 0 0>, rotation=<rot 90 0 0>}]
 dyn_hair_bk_r_03 [{shapeType=Capsule, length=1.3, radius=1.5, center=<2 0 0>, rotation=<rot 90 0 0>}]
 dyn_hair_fr_l_01 [{shapeType=Capsule, length=1.5, radius=1.5, center=<1.5 0 0>, rotation=<rot 90 0 0>}]
 dyn_hair_fr_l_02 [{shapeType=Capsule, length=1.5, radius=1.5, center=<2.2 0 0>, rotation=<rot 90 0 0>}]
 dyn_hair_fr_l_03 [{shapeType=Capsule, length=0.9, radius=1.2, center=<2 0 0>, rotation=<rot 90 0 0>}]
 dyn_hair_fr_r_01 [{shapeType=Capsule, length=2, radius=2, center=<2.2 0 0>, rotation=<rot 90 0 0>}]
 dyn_hair_fr_r_02 [{shapeType=Capsule, length=2.9347, radius=1.0653, center=<2.2 0 0>, rotation=<rot 80.5987 0 0>}]
 dyn_hair_fr_r_03 [{shapeType=Capsule, length=0.0144, radius=2, center=<1.6362 0 0>, rotation=<rot 90 0 0>}]
 dyn_hair_r_01 [{shapeType=Capsule, length=1.3, radius=1.5, center=<2.2 0 0>, rotation=<rot 90 0 0>}]
 dyn_hair_r_02 [{shapeType=Capsule, length=1.3, radius=1.5, center=<2.2 0 0>, rotation=<rot 90 0 0>}]
 dyn_hair_r_03 [{shapeType=Capsule, length=0.9, radius=1.5, center=<1.6 0 0>, rotation=<rot 90 0 0>}]
 dyn_hat_01 [{shapeType=Capsule, length=2, radius=2, center=<1.8408 0 0>, rotation=<rot 90 0 0>}]
 head [
  {shapeType=Capsule, length=4, radius=5, center=<0.0785 -1.25 0>, rotation=<rot 90 0 -10>}
  {shapeType=Capsule, length=6.1326, radius=5.8777, center=<2.1232 3.0881 0>, rotation=<rot 90 -90 -95>}
  {shapeType=Capsule, length=2.6955, radius=1.7232, center=<2.1169 -0.3915 6.1713>, rotation=<rot 85.9796 -174.9995 -180>}
  {shapeType=Capsule, length=2.6955, radius=1.7232, center=<2.0062 -0.5593 -6.3643>, rotation=<rot 81.8929 5.0021 0>}
  {shapeType=Capsule, length=2.75, radius=1.25, shapeName=head_capsule, center=<4.4717 7.9401 4.7546>, rotation=<rot 69.2629 -18.0007 -28.4494>}
 ]
 neck_01 [{shapeType=Capsule, length=5.9622, radius=6.4996, center=<2.0505 2.0367 0.0001>, rotation=<rot 90 0 0>}]
 spine_05 [{shapeType=Capsule, length=7.8353, radius=8.5564, center=<4.1179 -0.9018 0>}]
}
constraints [
 {bone1Name=dyn_bangs_l_01, bone2Name=head, swing1LimitDegrees=1, swing2LimitDegrees=1, twistLimitDegrees=1}
 {bone1Name=dyn_bangs_l_02, bone2Name=dyn_bangs_l_01, swing1LimitDegrees=1, swing2LimitDegrees=1, twistLimitDegrees=1}
 {bone1Name=dyn_hair_fr_l_01, bone2Name=head, swing1LimitDegrees=20, swing2LimitDegrees=45, twistLimitDegrees=0}
 {bone1Name=dyn_hair_fr_l_02, bone2Name=dyn_hair_fr_l_01, swing1LimitDegrees=20, swing2LimitDegrees=45, twistLimitDegrees=0}
 {bone1Name=dyn_hair_fr_l_03, bone2Name=dyn_hair_fr_l_02, swing1LimitDegrees=20, swing2LimitDegrees=45, twistLimitDegrees=0}
 {bone1Name=dyn_hair_bk_l_01, bone2Name=head, swing1LimitDegrees=20, swing2LimitDegrees=25, twistLimitDegrees=0}
 {bone1Name=dyn_hair_bk_l_02, bone2Name=dyn_hair_bk_l_01, swing1LimitDegrees=20, swing2LimitDegrees=25, twistLimitDegrees=0}
 {bone1Name=dyn_hair_bk_l_03, bone2Name=dyn_hair_bk_l_02, swing1LimitDegrees=20, swing2LimitDegrees=25, twistLimitDegrees=0}
 {bone1Name=dyn_hair_bk_01, bone2Name=head, swing1LimitDegrees=20, swing2LimitDegrees=25, twistLimitDegrees=0}
 {bone1Name=dyn_hair_bk_02, bone2Name=dyn_hair_bk_01, swing1LimitDegrees=20, swing2LimitDegrees=25, twistLimitDegrees=0}
 {bone1Name=dyn_hair_bk_03, bone2Name=dyn_hair_bk_02, swing1LimitDegrees=20, swing2LimitDegrees=25, twistLimitDegrees=0}
 {bone1Name=dyn_hat_01, bone2Name=head, swing1LimitDegrees=45, swing2LimitDegrees=45, twistLimitDegrees=5}
 {bone1Name=dyn_hair_r_01, bone2Name=head, swing1LimitDegrees=20, swing2LimitDegrees=25, twistLimitDegrees=0}
 {bone1Name=dyn_hair_r_02, bone2Name=dyn_hair_r_01, swing1LimitDegrees=20, swing2LimitDegrees=25, twistLimitDegrees=0}
 {bone1Name=dyn_hair_r_03, bone2Name=dyn_hair_r_02, swing1LimitDegrees=20, swing2LimitDegrees=25, twistLimitDegrees=0}
 {bone1Name=dyn_hair_bk_r_01, bone2Name=head, swing1LimitDegrees=20, swing2LimitDegrees=25, twistLimitDegrees=0}
 {bone1Name=dyn_hair_bk_r_02, bone2Name=dyn_hair_bk_r_01, swing1LimitDegrees=20, swing2LimitDegrees=25, twistLimitDegrees=0}
 {bone1Name=dyn_hair_bk_r_03, bone2Name=dyn_hair_bk_r_02, swing1LimitDegrees=20, swing2LimitDegrees=25, twistLimitDegrees=0}
 {bone1Name=dyn_hair_fr_r_01, bone2Name=head, swing1LimitDegrees=20, swing2LimitDegrees=45, twistLimitDegrees=5}
 {bone1Name=dyn_hair_fr_r_02, bone2Name=dyn_hair_fr_r_01, swing1LimitDegrees=20, swing2LimitDegrees=45, twistLimitDegrees=5}
 {bone1Name=dyn_hair_fr_r_03, bone2Name=dyn_hair_fr_r_02, swing1LimitDegrees=20, swing2LimitDegrees=45, twistLimitDegrees=5}
 {bone1Name=dyn_bangs_r_01, bone2Name=head, swing1LimitDegrees=1, swing2LimitDegrees=1, twistLimitDegrees=1}
 {bone1Name=dyn_bangs_r_02, bone2Name=dyn_bangs_r_01, swing1LimitDegrees=1, swing2LimitDegrees=1, twistLimitDegrees=1}
 {bone1Name=dyn_bangs_r_03, bone2Name=dyn_bangs_r_02, swing1LimitDegrees=1, swing2LimitDegrees=1, twistLimitDegrees=1}
 {
  bone1Name=head
  bone2Name=neck_01
  swing1LimitDegrees=45
  swing1Motion=Free
  swing2LimitDegrees=45
  swing2Motion=Free
  twistLimitDegrees=45
  twistMotion=Free
 }
 {
  bone1Name=clavicle_r
  bone2Name=spine_05
  swing1LimitDegrees=45
  swing1Motion=Free
  swing2LimitDegrees=45
  swing2Motion=Free
  twistLimitDegrees=45
  twistMotion=Free
 }
 {
  bone1Name=clavicle_l
  bone2Name=spine_05
  swing1LimitDegrees=45
  swing1Motion=Free
  swing2LimitDegrees=45
  swing2Motion=Free
  twistLimitDegrees=45
  twistMotion=Free
 }
 {
  bone1Name=neck_01
  bone2Name=spine_05
  swing1LimitDegrees=45
  swing1Motion=Free
  swing2LimitDegrees=45
  swing2Motion=Free
  twistLimitDegrees=45
  twistMotion=Free
 }
]
